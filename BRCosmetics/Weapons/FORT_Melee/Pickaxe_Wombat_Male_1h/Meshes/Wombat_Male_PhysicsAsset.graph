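# Physics asset for the Wombat Male one-handed pickaxe, a cosmetic melee weapon mesh.
# A single capsule body on the Melee_TwoHanded_Handle bone is the only collision shape, offset along Z from the bone origin.
# It has no constraints, so the asset defines no jointed or separately simulated parts.
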PhysicsAsset /BRCosmetics/Weapons/FORT_Melee/Pickaxe_Wombat_Male_1h/Meshes/Wombat_Male_PhysicsAsset
bodies {Melee_TwoHanded_Handle [{shapeType=Capsule, length=23.0119, radius=5.8491, center=<-0.0574 0.3587 13.4247>, rotation=<rot 4.4049 -90 0.01>}]}
constraints []
